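# Material function that paints a three-digit number, such as on a character cosmetic, over incoming material attributes.
# The clamped Number parameter is split by Fmod and division into hundreds, tens and ones, each picking a FlipBook cell from a digit atlas.
# Offset, scaled, rotated UV1 forms three side-by-side slots, max-combined and bounds-masked.
# The resulting alpha blends digit color into BaseColor, zeroes SubsurfaceColor and is output separately.
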MaterialFunction /Game/MF_Digits
graph {nodes [
  Constant()
  n1 = GetMaterialAttributes(None=FunctionInput() {inputName=Material Attributes, inputType=FunctionInput_MaterialAttributes}) {attributeGetTypes [#1, #2, #3]}
  n2 = VectorParameter() {group=Digits, parameterName=Digits Color, defaultValue=<rgba 0.887 0.8341 0.6919 1>}
  n3 = Multiply(
   A=LinearInterpolate(
    A=@n2.RGB
    B=FunctionInput(Preview=@n2.RGB) {bUsePreviewValueAsDefault=true, inputName=Digits Color, sortPriority=1}
    Alpha=ScalarParameter() {group=Digits, parameterName=Override Digits Color}
   )
   B=ScalarParameter() {defaultValue=0.288, group=Digits, parameterName=Digits Intensity}
  )
  n4 = Clamp(
   None=ScalarParameter() {defaultValue=7, group=Digits, parameterName=Number}
   Min=ScalarParameter() {defaultValue=1, group=Digits, parameterName=Number_Min, sortPriority=33}
   Max=ScalarParameter() {defaultValue=456, group=Digits, parameterName=Number_Max, sortPriority=34}
  ) {maxDefault=999}
  n5 = Constant() {r=2}
  n6 = Constant() {r=5}
  n7 = VectorParameter() {parameterName=Digits Position}
  n8 = MaterialFunctionCall(
   UVs=Add(A=TextureCoordinate() {coordinateIndex=1}, B=ComponentMask(None=@n7.RGB) {g=true, r=true})
   Texture Scale=ComponentMask(None=VectorParameter() {group=Digits, parameterName=Digits Scale, defaultValue=<rgba 1 1 0 0>}.RGB) {g=true, r=true}
  ) {materialFunction=&ScaleUVsByCenter}
  n9 = Multiply(
   A=MaterialFunctionCall(UVs=@n8.UVs, "Rotation Angle (0-1)"=ScalarParameter() {group=Digits, parameterName=Rotation Angle}) {materialFunction=&CustomRotator}.Rotated Values
   B=Constant2Vector() {g=1, r=3}
  ) {constB=3}
  n10 = Saturate(None=@n9)
  n11 = Constant2Vector() {g=1, r=1.2}
  n12 = TextureSampleParameter2D(UVs=MaterialFunctionCall(
    "Animation  Phase (0-1)"=Divide(A=@n4) {constB=1000}
    Number of Rows=@n5
    Number of Columns=@n6
    UVs=MaterialFunctionCall(UVs=@n10, Texture Scale=@n11) {materialFunction=&ScaleUVsByCenter}.UVs
   ) {materialFunction=&FlipBook}.UVs) {
   group=Digits
   parameterName=Mask_Digits
   samplerType=SAMPLERTYPE_Masks
   texture=/BRCosmetics/Characters/Player/Male/Medium/Bodies/M_MED_TimberStakeDrift/Textures/T_TimberStakeDrift_Numbers
  }
  n13 = Fmod(A=@n4, B=Constant() {r=100})
  n14 = Subtract(A=@n9, B=Constant2Vector() {r=1})
  n15 = Saturate(None=@n14)
  n16 = TextureSampleParameter2D(UVs=MaterialFunctionCall(
    "Animation  Phase (0-1)"=Divide(A=@n13) {constB=100}
    Number of Rows=@n5
    Number of Columns=@n6
    UVs=MaterialFunctionCall(UVs=@n15, Texture Scale=@n11) {materialFunction=&ScaleUVsByCenter}.UVs
   ) {materialFunction=&FlipBook}.UVs) {
   group=Digits
   parameterName=Mask_Digits
   samplerType=SAMPLERTYPE_Masks
   texture=&T_TimberStakeDrift_Numbers
  }
  n17 = Fmod(A=@n4, B=Constant() {r=10})
  n18 = Saturate(None=Subtract(A=@n14, B=Constant2Vector() {r=1}))
  n19 = TextureSampleParameter2D(UVs=MaterialFunctionCall(
    "Animation  Phase (0-1)"=Divide(A=@n17) {constB=10}
    Number of Rows=@n5
    Number of Columns=@n6
    UVs=MaterialFunctionCall(UVs=@n18, Texture Scale=@n11) {materialFunction=&ScaleUVsByCenter}.UVs
   ) {materialFunction=&FlipBook}.UVs) {
   group=Digits
   parameterName=Mask_Digits
   samplerType=SAMPLERTYPE_Masks
   texture=&T_TimberStakeDrift_Numbers
  }
  n20 = Max(A=@n16.R, B=@n19.R)
  n21 = Multiply(
   A=Multiply(A=Max(A=@n12.R, B=@n20), B=NamedRerouteUsage())
   B=ScalarParameter() {defaultValue=1, group=Digits, parameterName=Digits_AlphaStrength, sortPriority=15}
  )
  FunctionOutput(None=SetMaterialAttributes!(
    MaterialAttributes=@n1.MaterialAttributes
    BaseColor=LinearInterpolate(A=@n1.BaseColor, B=@n3, Alpha=@n21)
    SubsurfaceColor=LinearInterpolate(A=@n1.SubsurfaceColor, Alpha=@n21) {constB=0}
   ) {attributeSetTypes [#1, #3]})
  FunctionOutput(None=@n21) {outputName=Numbers Alpha}
  NamedRerouteDeclaration(None=@n8.0-1 mask) {name=NumbersMask, nodeColor=<rgba 0 0.0625 1 1>}
 ]}
