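# Parent material for the VigorKnack pickaxe cosmetics, adding recolorable, reactive primary and secondary regions to the shared character shading.
# Between CharacterParent PreFX and PostFX functions, mask channels blend in tinted grayscale color or emissive, with optional secondary clear coat and a digits overlay.
# Reactivity parameters drive color boost, fresnel-modulated glow and desaturation, with variant switches for Dill, Golf and black versions.
# It also defines optional scanlines and a temporal responsiveness output.
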Material /BRCosmetics/Weapons/FORT_Melee/Pickaxe_VigorKnackDill/Materials/M_VigorKnack_Parent
graph {
 properties {
  bAutomaticallySetUsageInEditor=false
  bUseMaterialAttributes=true
  bUsedWithMeshDeformer=true
  bUsedWithMorphTargets=true
  bUsedWithSkeletalMesh=true
  pixelDepthOffsetMode=PDOM_Legacy
  previewMesh=/BRCosmetics/Accessories/FORT_Backpacks/Backpack_M_MED_VigorKnackDill/Meshes/M_MED_VigorKnackDill_Pack
  shadingModel=MSM_FromMaterialExpression
  usedShadingModels=Default Lit | Subsurface | Clear Coat | Subsurface Profile | Cloth
 }
 nodes [
  n1 = MaterialFunctionCall() {materialFunction=&MF_CharacterParent_2023_PreFX}
  n2 = GetMaterialAttributes(None=@n1.Result) {attributeGetTypes [
    #1
    #2
    #3
    #4
   ]}
  ClearCoatNormalCustomOutput!(None=@n2.Normal)
  Constant3Vector() {constant=<rgba 0.5 0.5 1 0>}
  LinearInterpolate()
  LinearInterpolate()
  n3 = NamedRerouteUsage()
  n4 = NamedRerouteUsage()
  n5 = ScalarParameter() {defaultValue=1, parameterName=Primary Color Exp}
  n6 = ScalarParameter() {defaultValue=1, parameterName=Primary Color Influence}
  n7 = LinearInterpolate(
   A=ScalarParameter() {defaultValue=1, parameterName=Primary Color - Second Influence}
   B=ScalarParameter() {defaultValue=50, parameterName=Primary Color - Second Influence Jeans}
   Alpha=NamedRerouteUsage()
  )
  n8 = ScalarParameter() {controlType=Enumeration, enumeration=/Game/Global/Materials/Enums/E_Mat_OnOff, parameterName=Primary Color - Use Second Influence}
  n9 = Multiply(A=Multiply(A=Power(Base=@n4, Exp=@n5), B=LinearInterpolate(A=@n6, B=@n7, Alpha=@n8)), B=NamedRerouteUsage())
  n10 = NamedRerouteUsage()
  n11 = ScalarParameter() {parameterName=Use Emissive as Primary Color}
  n12 = NamedRerouteUsage()
  n13 = Multiply(A=Desaturation(None=@n2.Emissive Color), B=@n12)
  n14 = Multiply(A=@n2.Base Color, B=ScalarParameter() {defaultValue=2, parameterName=GolfPickaxeEmissiveStrength}) {constB=2}
  n15 = SetMaterialAttributes!(
   MaterialAttributes=@n2.MaterialAttributes
   Base Color=LinearInterpolate(A=LinearInterpolate(A=@n3, B=@n9, Alpha=@n10), Alpha=@n11) {constB=0}
   Fuzz Color or Subsurface Color
   Specular=Multiply(A=@n2.Specular, B=ScalarParameter() {defaultValue=1, parameterName=Primary - Specular Override})
   Shading Model
   Emissive Color=LinearInterpolate(
    A=LinearInterpolate(A=@n2.Emissive Color, B=@n13, Alpha=@n11)
    B=Multiply(A=@n14, B=@n12)
    Alpha=ScalarParameter() {parameterName=isGolfPickaxe}
   )
  ) {attributeSetTypes [
    #1
    #5
    #3
    #6
    #4
   ]}
  n16 = BlendMaterialAttributes(A=@n2.MaterialAttributes, B=@n15, Alpha=NamedRerouteUsage())
  n17 = StaticSwitchParameter(True=@n16, False=@n2.MaterialAttributes) {defaultValue=true, parameterName=Use Primary}
  n18 = NamedRerouteUsage()
  n19 = MaterialFunctionCall!(Linear=@n18) {materialFunction=/Engine/Functions/Engine_MaterialFunctions01/ImageAdjustment/LinearTosRGB}
  n20 = ScalarParameter() {parameterName=sRGBConversion}
  n21 = NamedRerouteUsage()
  n22 = Max(
   A=LinearInterpolate(A=LinearInterpolate(A=@n18, B=@n19.sRGB, Alpha=@n20), B=OneMinus(None=@n21), Alpha=ScalarParameter() {parameterName=Inverted Grayscale})
   B=ScalarParameter() {parameterName=Secondaryy Color - Low Level}
  )
  n23 = ScalarParameter() {defaultValue=1, parameterName=Secondary Color Power}
  n24 = ScalarParameter() {defaultValue=1, parameterName=Secondary Color Influence}
  n25 = LinearInterpolate(
   A=ScalarParameter() {parameterName=Secondary Color - Secondary Influence}
   B=ScalarParameter() {defaultValue=50, parameterName=Secondary Color - Second Influence Jeans}
   Alpha=NamedRerouteUsage()
  )
  n26 = ScalarParameter() {controlType=Enumeration, enumeration=&E_Mat_OnOff, parameterName=Secondary Color - Use Secondary Influence}
  n27 = ScalarParameter() {defaultValue=1, parameterName=Secondary Color - Color Boost}
  n28 = Add(A=@n27, B=Multiply(A=ScalarParameter() {defaultValue=0.08, parameterName=ColorBoostAddReactive}, B=NamedRerouteUsage())) {constB=0.05}
  n29 = ScalarParameter() {parameterName=Reactivity Intensity}
  n30 = Multiply(
   A=Multiply(A=Power(Base=@n22, Exp=@n23), B=LinearInterpolate(A=@n24, B=@n25, Alpha=@n26))
   B=Multiply(
    A=VectorParameter() {parameterName=Secondary Color, defaultValue=<rgba 0.5 0.8323 0.017 1>}.RGB
    B=LinearInterpolate(A=@n27, B=@n28, Alpha=@n29)
   )
  )
  n31 = ScalarParameter() {controlType=Enumeration, defaultValue=1, enumeration=&E_Mat_OnOff, parameterName=Use Secondary As Emissive Color}
  n32 = TextureSampleParameter2D(UVs=Multiply(A=TextureCoordinate(), B=ScalarParameter() {defaultValue=4.2288, parameterName=Clear Coat Normal})) {parameterName=Clear Coat Normal, samplerType=SAMPLERTYPE_Normal, texture=&T_Weap_Distort_N_02}
  n33 = MaterialFunctionCall(Normal=@n32.RGB, Flatness=ScalarParameter() {parameterName=Clear Coat Normal Flatten}) {materialFunction=&FlattenNormal}
  n34 = TextureSampleParameter2D(UVs=Multiply(A=TextureCoordinate(), B=ScalarParameter() {defaultValue=6.8471, parameterName=Detail Normal Tiling})) {parameterName=DetailNormal, samplerType=SAMPLERTYPE_Normal, texture=/Game/Global/Textures/Default/Tileable_Patterns/T_WaterSoft_N}
  n35 = MaterialFunctionCall(Normal=@n34.RGB, Flatness=ScalarParameter() {defaultValue=0.7494, parameterName=Detail Normal Flatten}) {materialFunction=&FlattenNormal}
  n36 = SetMaterialAttributes!(
   MaterialAttributes=@n17
   Base Color=LinearInterpolate(A=@n30, Alpha=@n31) {constB=0}
   Shading Model=StaticSwitchParameter(True=ShadingModel() {shadingModel=MSM_ClearCoat}, False=ShadingModel()) {parameterName=Secondary - Use Clear Coat}
   Normal=StaticSwitchParameter(
    True=MaterialFunctionCall(BaseNormal=@n33.Result, AdditionalNormal=@n35.Result) {materialFunction=&BlendAngleCorrectedNormals}.Result
    False=@n2.Normal
   ) {parameterName=Secondary - Use Clear Coat}
   Clear Coat or Cloth or Curvature or Custom Data 0=ScalarParameter() {defaultValue=1, parameterName=Clear Coat}
   Clear Coat Roughness or Custom Data 1=ScalarParameter() {defaultValue=0.1, parameterName=Clear Coat Roughness}
   Specular=Multiply(A=@n2.Specular, B=ScalarParameter() {defaultValue=1, parameterName=Secondary - Specular Override})
   Emissive Color=Multiply(A=LinearInterpolate(B=@n30, Alpha=@n31), B=NamedRerouteUsage())
   Fuzz Color or Subsurface Color
  ) {attributeSetTypes [
    #1
    #6
    #2
    #7
    #8
    #3
    #4
    #5
   ]}
  n37 = StaticSwitchParameter(True=BlendMaterialAttributes(A=@n16, B=@n36, Alpha=NamedRerouteUsage()), False=@n17) {defaultValue=true, parameterName=Use Secondary}
  n38 = NamedRerouteUsage()
  n39 = MaterialFunctionCall!(
   Material Attributes=@n37
   Digits Color=LinearInterpolate(A=Multiply(A=@n38) {constB=0.35}, B=@n38, Alpha=Fresnel() {baseReflectFraction=0.06, exponent=1})
  ) {materialFunction=/Game/MF_Digits}
  n40 = StaticSwitchParameter(True=@n39.Result, False=@n37) {parameterName=Use Digits}
  n41 = GetMaterialAttributes(None=@n40) {attributeGetTypes [#4, #1]}
  n42 = NamedRerouteUsage()
  n43 = TextureSampleParameter2D() {
   parameterName=DarkVariantEmissive
   texture=/BRCosmetics/Characters/Player/Male/Medium/Bodies/M_MED_VigorKnackDill/Textures/T_VigorKnackDill_Body_E
  }
  n44 = Multiply(A=@n43.RGB) {constB=0.3}
  n45 = NamedRerouteUsage()
  n46 = NamedRerouteUsage()
  n47 = ScalarParameter() {controlType=Enumeration, enumeration=&E_Mat_OnOff, parameterName=isBlackPrimary}
  n48 = NamedRerouteUsage()
  n49 = ScalarParameter() {controlType=Enumeration, enumeration=&E_Mat_OnOff, parameterName=isBlackSecondary}
  n50 = Multiply(A=Add(A=Multiply(A=@n46, B=@n47), B=Multiply(A=@n48, B=@n49)), B=NamedRerouteUsage())
  n51 = ScalarParameter() {parameterName=hasBlackVariant}
  n52 = LinearInterpolate(A=@n41.Base Color, B=Add(A=Multiply(A=@n42, B=@n44), B=Multiply(A=@n44, B=@n45)), Alpha=Saturate(None=Multiply(A=@n50, B=@n51)))
  n53 = ScalarParameter() {
   controlType=Enumeration
   defaultValue=1
   enumeration=&E_Mat_OnOff
   parameterName=Use Base Color for Emissive Reactivity
  }
  n54 = Multiply(A=LinearInterpolate(A=@n41.Emissive Color, B=@n52, Alpha=@n53), B=NamedRerouteUsage())
  n55 = Multiply(A=@n54, B=ScalarParameter() {parameterName=Reactivity Intensity})
  n56 = ScalarParameter() {parameterName=isGolfPickaxe}
  n57 = LinearInterpolate(
   A=@n41.Emissive Color
   B=LinearInterpolate(
    A=@n54
    B=LinearInterpolate(A=@n55, B=@n54, Alpha=@n56)
    Alpha=ScalarParameter() {controlType=Enumeration, enumeration=&E_Mat_OnOff, parameterName=IsVigorKnackDill}
   )
   Alpha=ScalarParameter() {parameterName=Is Reactive}
  )
  n58 = NamedRerouteUsage()
  n59 = LinearInterpolate(
   A=@n41.Base Color
   B=LinearInterpolate(A=@n41.Base Color, B=@n52, Alpha=Saturate(None=@n58)) {constB=0}
   Alpha=ScalarParameter() {parameterName=Is Reactive}
  )
  n60 = LinearInterpolate(
   A=ScalarParameter() {parameterName=DesaturationDillPrimary}
   B=ScalarParameter() {parameterName=DesaturationDillSecondary}
   Alpha=NamedRerouteUsage()
  )
  n61 = Multiply(A=NamedRerouteUsage(), B=NamedRerouteUsage())
  n62 = LinearInterpolate(A=Desaturation(None=@n59, Fraction=Multiply(A=@n60, B=@n61)), B=@n59, Alpha=ScalarParameter() {parameterName=Reactivity Intensity})
  n63 = NamedRerouteUsage()
  n64 = Max(A=StaticSwitchParameter(True=NamedRerouteUsage(), False=Multiply(A=@n63) {constB=0}) {parameterName=Use Digits}, B=NamedRerouteUsage())
  MaterialFunctionCall(
   Material=BlendMaterialAttributes(
    A=@n40
    B=SetMaterialAttributes!(MaterialAttributes=@n41.MaterialAttributes, Emissive Color=@n57, Base Color=@n62) {attributeSetTypes [#4, #1]}
    Alpha=Saturate(None=@n64)
   )
   M=@n1.M
   Pre Skinned Local Position=@n1.M
   Distance Blend=@n1.M
   Fresnel=@n1.M
  ) {materialFunction=&MF_CharacterParent_2023_PostFX}
  n65 = ScalarParameter() {defaultValue=0.009, parameterName=Scanline Speed}
  n66 = TextureCoordinate()
  NamedRerouteDeclaration(None=LinearInterpolate(
    B=MaterialFunctionCall!(
     Number of Scanlines
     Scanline Speed=@n65
     "Scanlines Thickness (Bias)"
     Scanlines Dark Value
     Scanlines Light Value
     Block Opacity
     UVs=@n66
     Time
     Block Repeat
    ) {materialFunction=/Game/Effects/Fort_Effects/Effects/Atlas/Materials/MF_Atlas_Scanlines_01}.Result
    Alpha=ScalarParameter() {parameterName=Use Scanlines}
   ) {constA=1}) {name=Scanlines, nodeColor=<rgba 0.5 0 1 1>}
  n67 = TextureSampleParameter2D!(UVs, Bias=ScalarParameter() {parameterName=TA Mask - Mip Bias}, Apply View MipBias) {
   mipValueMode=TMVM_MipBias
   parameterName=Color Customization Mask
   samplerType=SAMPLERTYPE_Masks
   texture=/BRCosmetics/Characters/Player/Male/Medium/Bodies/M_MED_VigorKnackGolf/Textures/T_VigorKnackGolf_Body_FX1
  }
  n68 = StaticSwitchParameter(True=@n67.RGB, False=@n1.Tech Art Mask) {parameterName=Use Color Customization Mask}
  n69 = StaticComponentMaskParameter(None=@n68) {defaultG=true, parameterName=Secondary Mask Channel}
  NamedRerouteDeclaration(None=@n69) {name=Secondary Mask, nodeColor=<rgba 0 0.2266 1 1>}
  n70 = LinearInterpolate(
   A=ScalarParameter() {defaultValue=5, parameterName=Primary - Reactivity Max}
   B=ScalarParameter() {defaultValue=5, parameterName=Secondary - Reactivity Max}
   Alpha=NamedRerouteUsage()
  )
  n71 = ScalarParameter() {parameterName=Reactivity Intensity}
  n72 = ScalarParameter() {defaultValue=0.25, parameterName=Reactivity - Fresnel Exp}
  n73 = ScalarParameter() {parameterName=Reactivity Intensity}
  n74 = OneMinus(None=Fresnel(ExponentIn=LinearInterpolate(B=@n72, Alpha=@n73) {constB=0}) {baseReflectFraction=0})
  n75 = ScalarParameter() {defaultValue=1, parameterName=Reactivity - Use Fresnel}
  NamedRerouteDeclaration(None=Multiply(A=LinearInterpolate(B=@n70, Alpha=@n71) {constA=1, constB=1.5}, B=LinearInterpolate(B=@n74, Alpha=@n75) {constA=1})) {name=ReactivityFresnel, nodeColor=<rgba 0 1 0.9531 1>}
  n76 = StaticComponentMaskParameter(None=@n68) {defaultR=true, parameterName=Primary Color - Mask}
  NamedRerouteDeclaration(None=Max(A=@n76, B=@n69)) {name=Color Masks, nodeColor=<rgba 0 1 0.5312 1>}
  n77 = VectorParameter() {parameterName=Primary Color, defaultValue=<rgba 0.0704 0 0.1559 1>}
  NamedRerouteDeclaration(None=@n77.RGB) {name=PrimaryColor, nodeColor=<rgba 0 0.6484 1 1>}
  NamedRerouteDeclaration(None=@n39.Numbers Alpha) {name=DigitsAlpha, nodeColor=<rgba 0.9688 1 0 1>}
  NamedRerouteDeclaration(None=Desaturation(None=@n2.Base Color)) {name=GreyBaseColor, nodeColor=<rgba 1 0.9844 0 1>}
  NamedRerouteDeclaration(None=@n76) {name=Primary Mask, nodeColor=<rgba 0.6641 0 1 1>}
  n78 = ScalarParameter() {defaultValue=1, parameterName=Primary - Color Boost}
  n79 = Add(A=@n78, B=Multiply(A=ScalarParameter() {defaultValue=0.08, parameterName=ColorBoostAddReactive}, B=NamedRerouteUsage())) {constB=0.05}
  n80 = ScalarParameter() {parameterName=Reactivity Intensity}
  NamedRerouteDeclaration(None=Multiply(A=@n77.RGB, B=LinearInterpolate(A=@n78, B=@n79, Alpha=@n80))) {name=PrimaryColorBoosted, nodeColor=<rgba 0.1016 0 1 1>}
  NamedRerouteDeclaration(None=@n67.B) {name=DesaturationMask, nodeColor=<rgba 1 0.7734 0 1>}
  NamedRerouteDeclaration(None=Saturate(None=Subtract(A=@n76, B=@n69))) {name=PrimaryMaskBlack, nodeColor=<rgba 1 0 0.3984 1>}
  n81 = NamedRerouteUsage()
  n82 = ScalarParameter() {parameterName=Temporal Responsiveness}
  TemporalResponsivenessOutput!(None=LinearInterpolate(Alpha=Multiply(A=@n81, B=@n82)))
 ]
}
